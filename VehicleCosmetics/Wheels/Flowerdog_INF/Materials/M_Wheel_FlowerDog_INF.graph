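# Unlit skeletal-mesh material for the Flowerdog INF wheel cosmetic, outputting only an animated emissive glow.
# MF_Infinity_OffsetFunction builds infinity-effect masks, alongside a scrolling "Mandela" grid with separate rim and tread tiling and a BumpOffset parallax echo.
# A noise-distorted warp mask, a panning two-color blend and an advanced Fresnel edge add motion, tint and rim light.
# Named reroutes, including vertex-color red and packed texture masks, carry the layers toward the emissive composite.
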Material /VehicleCosmetics/Wheels/Flowerdog_INF/Materials/M_Wheel_FlowerDog_INF
graph {
 properties {
  bAutomaticallySetUsageInEditor=false
  bUseMaterialAttributes=true
  bUsedWithMeshDeformer=true
  bUsedWithSkeletalMesh=true
  shadingModel=MSM_Unlit
  usedShadingModels=Unlit
 }
 nodes [
  n1 = Clamp(None=Add(A=NamedRerouteUsage(), B=NamedRerouteUsage()))
  n2 = NamedRerouteUsage()
  n3 = NamedRerouteUsage()
  n4 = Add(
   A=Add(A=NamedRerouteUsage(), B=NamedRerouteUsage())
   B=Multiply(A=TextureSample(Tex=@n3) {samplerSource=SSM_Wrap_WorldGroupSettings, texture=/VehicleCosmetics/Wheels/Flowerdog_INF/Textures/T_Flowerdog_infinity_Grid_RGB}.A) {constB=0.5}
  )
  n5 = NamedRerouteUsage()
  n6 = Multiply(
   A=Multiply(
    A=VectorParameter() {group=01 Grid, parameterName=Grid Color, sortPriority=-1, defaultValue=<rgba 0.1529 0.3185 1 1>}.RGB
    B=NamedRerouteUsage()
   )
   B=ScalarParameter() {defaultValue=1, group=01 Grid, parameterName=Grid Multiply, sortPriority=-1}
  )
  n7 = OneMinus(None=NamedRerouteUsage())
  n8 = LinearInterpolate(A=Add(A=Multiply(A=@n1, B=@n2)) {constB=0}, B=Add(A=Multiply(A=@n4, B=@n5), B=Multiply(A=@n6, B=@n7)), Alpha=NamedRerouteUsage())
  n9 = ScalarParameter() {defaultValue=2, group=00 Base, parameterName=Emissive Multiply}
  NamedRerouteDeclaration(None=Add(A=Multiply(A=@n8, B=@n9), B=NamedRerouteUsage())) {name=Emissive, nodeColor=<rgba 1 0 0.8672 1>}
  NamedRerouteDeclaration(None=VertexColor().R) {name=VC_R, nodeColor=<rgba 0 0.4609 1 1>}
  n10 = TextureObjectParameter() {group=00 Base, parameterName=Masks INF, sortPriority=2, texture=/VehicleCosmetics/Wheels/Flowerdog_INF/Textures/T_Flowerdog_infinity_Rim_RGB}
  n11 = NamedRerouteDeclaration(None=@n10) {name=Masks INF, nodeColor=<rgba 0.875 1 0 1>}
  n12 = ScalarParameter() {defaultValue=0.2, group=04 Infinity, parameterName=Infinity Offset Rim, sortPriority=2}
  n13 = ScalarParameter() {defaultValue=0.05, group=04 Infinity, parameterName=Infinity Gradient Darkness Rim, sortPriority=3}
  NamedRerouteDeclaration(None=ComponentMask(None=MaterialFunctionCall!(Texture=@n11, Offset=@n12, UVs, RotateUVs, GradientFalloff, GradientDarkness=@n13) {materialFunction=/VehicleCosmetics/Wheels/ClockGear_SE/Materials/MF_Infinity_OffsetFunction}.Result) {r=true}) {name=Rim INF, nodeColor=<rgba 0.8047 0 1 1>}
  n14 = MaterialFunctionCall(
   Texture=NamedRerouteUsage()
   Offset=ScalarParameter() {defaultValue=0.2, group=04 Infinity, parameterName=Infinity Offset, sortPriority=0}
   GradientDarkness=ScalarParameter() {defaultValue=0.05, group=04 Infinity, parameterName=Infinity Gradient Darkness, sortPriority=1}
  ) {materialFunction=&MF_Infinity_OffsetFunction}
  NamedRerouteDeclaration(None=Multiply(A=ComponentMask(None=@n14.Result) {r=true}, B=NamedRerouteUsage())) {name=INF, nodeColor=<rgba 0 1 0.4844 1>}
  n15 = TextureSample(Tex=NamedRerouteUsage()) {samplerSource=SSM_Wrap_WorldGroupSettings, texture=&T_Flowerdog_infinity_Grid_RGB}
  NamedRerouteDeclaration(None=@n15.R) {name=Masks INF R, nodeColor=<rgba 0 1 0.5781 1>}
  NamedRerouteDeclaration(None=@n15.G) {name=Masks INF G, nodeColor=<rgba 0 1 0 1>}
  NamedRerouteDeclaration(None=@n15.B) {name=Masks INF B, nodeColor=<rgba 0 0 1 1>}
  n16 = ScalarParameter() {defaultValue=0.2, group=01 Grid, parameterName=Grid Bump Offset Fade, sortPriority=6}
  n17 = TextureCoordinate() {coordinateIndex=1, desc=Rim}
  n18 = Multiply(
   A=ScalarParameter() {defaultValue=20, group=01 Grid, parameterName=Mandela Rim Tiling U, sortPriority=0}
   B=ComponentMask(None=@n17) {r=true}
  ) {constB=6}
  n19 = Multiply(A=ScalarParameter() {defaultValue=6, group=01 Grid, parameterName=Mandela Rim Tiling V, sortPriority=1}, B=ComponentMask(None=@n17) {g=true}) {constB=6}
  n20 = ComponentMask(None=Panner(Coordinate=TextureCoordinate() {coordinateIndex=1, vTiling=0.5}, Time=Constant() {r=-0.15}) {speedY=1}) {g=true}
  n21 = TextureCoordinate() {coordinateIndex=1, desc=Tire}
  n22 = Multiply(
   A=Multiply(A=OneMinus(None=@n20)) {constB=6}
   B=Multiply(
    A=ScalarParameter() {defaultValue=6, group=01 Grid, parameterName=Mandela Tread Tiling V, sortPriority=3}
    B=ComponentMask(None=@n21) {g=true}
   ) {constB=6}
  ) {constB=6}
  n23 = Multiply(
   A=ScalarParameter() {defaultValue=20, group=01 Grid, parameterName=Mandela Tread Tiling U, sortPriority=2}
   B=ComponentMask(None=@n21) {r=true}
  ) {constB=6}
  n24 = Panner(
   Coordinate=LinearInterpolate(A=AppendVector(A=@n18, B=@n19), B=AppendVector(A=@n23, B=@n22), Alpha=NamedRerouteUsage())
   Time=Multiply(A=Time(), B=ScalarParameter() {defaultValue=0.5, group=01 Grid, parameterName=Mandela Scroll Speed, sortPriority=4})
  ) {speedY=2}
  n25 = TextureObjectParameter() {
   group=00 Base
   parameterName=Grid Mask
   sortPriority=0
   texture=&T_Flowerdog_infinity_Grid_RGB
  }
  n26 = NamedRerouteDeclaration(None=@n25) {name=Grid Mask, nodeColor=<rgba 1 0 0.4687 1>}
  n27 = TextureSample(
   UVs=BumpOffset(Coordinate=@n24, Height=ScalarParameter() {defaultValue=-10, group=01 Grid, parameterName=Grid Bump Offset, sortPriority=5})
   Tex=@n26
  ) {samplerSource=SSM_Wrap_WorldGroupSettings, texture=&T_Flowerdog_infinity_Grid_RGB}
  NamedRerouteDeclaration(None=Add(
    A=Multiply(A=@n16, B=@n27.R)
    B=TextureSample(UVs=@n24, Tex=@n26) {samplerSource=SSM_Wrap_WorldGroupSettings, texture=&T_Flowerdog_infinity_Grid_RGB}.R
   )) {name=Grid, nodeColor=<rgba 0.1484 1 0 1>}
  n28 = Panner(Coordinate=TextureCoordinate() {coordinateIndex=1}, Time=Constant() {r=-1.3}) {speedY=0.1}
  n29 = TextureCoordinate() {coordinateIndex=1, desc=Rim UV, uTiling=0.2, vTiling=0.2}
  n30 = TextureCoordinate() {coordinateIndex=1, desc=Tire UV, vTiling=0.25}
  n31 = NamedRerouteUsage()
  n32 = Time()
  n33 = ScalarParameter() {defaultValue=1.3, group=01 Grid, parameterName=Grid Warp Scroll Speed, sortPriority=0}
  n34 = TextureSample(
   UVs=Panner(Coordinate=LinearInterpolate(A=@n29, B=@n30, Alpha=@n31), Time=Multiply(A=@n32, B=@n33)) {speedX=0.05, speedY=0.1}
   Tex=NamedRerouteUsage()
  ) {samplerSource=SSM_Wrap_WorldGroupSettings, texture=&T_Flowerdog_infinity_Grid_RGB}
  n35 = Multiply(A=OneMinus(None=@n34.A)) {constB=0.8}
  n36 = Clamp(None=TextureSample(UVs=Subtract(A=@n28, B=@n35), Tex=NamedRerouteUsage()) {samplerSource=SSM_Wrap_WorldGroupSettings, texture=&T_Flowerdog_infinity_Grid_RGB}.B)
  n37 = ConstantBiasScale(None=@n36) {bias=-0.7, scale=2}
  n38 = TextureCoordinate() {coordinateIndex=1}
  n39 = Clamp(None=ConstantBiasScale(None=ComponentMask(None=@n38) {g=true}) {bias=-0.5, scale=1.5})
  n40 = Multiply(A=Multiply(A=Clamp(None=@n37)) {constB=3}, B=@n39) {constB=3}
  n41 = ConstantBiasScale(None=@n34.A) {bias=-0.3, scale=1.5}
  n42 = Clamp(None=Add(A=@n36, B=Clamp(None=@n41)))
  n43 = Multiply(A=Add(A=@n40, B=Multiply(A=@n39, B=@n42)), B=NamedRerouteUsage()) {constB=3}
  NamedRerouteDeclaration(None=Clamp(None=Add(A=@n40, B=@n43))) {name=Warp, nodeColor=<rgba 1 0.8906 0 1>}
  n44 = VectorParameter() {group=02 Color, parameterName=Custom Color 2, sortPriority=6, defaultValue=<rgba 1 0.8879 0.7379 1>}
  n45 = OneMinus(None=ScalarParameter() {group=02 Color, parameterName=Darken Custom Color 2})
  n46 = Multiply(A=Time(), B=ScalarParameter() {defaultValue=0.2, group=02 Color, parameterName=Color Pan Speed})
  n47 = TextureObjectParameter() {group=00 Base, parameterName=Mask, sortPriority=1, texture=/VehicleCosmetics/Wheels/Flowerdog_INF/Textures/T_Flowerdog_infinity_Rim_RGB2}
  n48 = NamedRerouteDeclaration(None=@n47) {name=Mask, nodeColor=<rgba 1 0 0.4688 1>}
  n49 = TextureSample(UVs=Panner(Coordinate=TextureCoordinate() {coordinateIndex=1}, Time=@n46) {speedY=1}, Tex=@n48) {samplerSource=SSM_Wrap_WorldGroupSettings, texture=&T_Flowerdog_infinity_Grid_RGB}
  n50 = TextureSample(UVs=Panner(Coordinate=TextureCoordinate() {vTiling=0.2}, Time=@n46) {speedY=1}, Tex=@n48) {samplerSource=SSM_Wrap_WorldGroupSettings, texture=&T_Flowerdog_infinity_Grid_RGB}
  n51 = NamedRerouteUsage()
  NamedRerouteDeclaration(None=LinearInterpolate(
    A=VectorParameter() {group=02 Color, parameterName=Custom Color, sortPriority=5, defaultValue=<rgba 1 0.2874 0 1>}.RGB
    B=Multiply(A=@n44.RGB, B=@n45)
    Alpha=LinearInterpolate(A=@n49.B, B=@n50.B, Alpha=@n51)
   )) {name=Color, nodeColor=<rgba 0 1 0.25 1>}
  n52 = ScalarParameter() {defaultValue=1.5, group=03 Fresnel, parameterName=Fresnel Sharpness, sortPriority=1}
  n53 = Multiply(
   A=Multiply(
    A=NamedRerouteUsage()
    B=MaterialFunctionCall!(Input Normal, Input Sharpness=@n52, Input Two Sided) {materialFunction=/VehicleCosmetics/Wheels/CarCar_Holographic/Materials/MF_Fresnel_Advanced}.Edge
   )
   B=ScalarParameter() {defaultValue=2, group=03 Fresnel, parameterName=Fresnel Strength, sortPriority=0}
  )
  n54 = OneMinus(None=NamedRerouteUsage())
  NamedRerouteDeclaration(None=LinearInterpolate(A=@n53, B=Multiply(A=@n53, B=@n54), Alpha=NamedRerouteUsage())) {name=Fresnel, nodeColor=<rgba 1 0 0.0703 1>}
  SetMaterialAttributes(Emissive Color=NamedRerouteUsage()) {attributeSetTypes [#1]}
 ]
}
